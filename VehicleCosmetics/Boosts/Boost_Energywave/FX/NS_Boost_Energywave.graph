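# Niagara effect for the Energywave vehicle boost cosmetic, trailing energy from the car's boost sockets.
# Core and spiral ribbon emitters spawn per socket at quality-scaled rates, stream backward and layer several ribbon renderers.
# Two near-identical sprite emitters partition streak particles across sockets, swirling them with vortex and curl-noise forces.
# Painted-variant overrides set colors, bIsPainted switches glow or brightness, and frontend-preview and car-offset parameters adjust placement.
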NiagaraSystem /VehicleCosmetics/Boosts/Boost_Energywave/FX/NS_Boost_Energywave
dependencies {
 usedDataInterfaces [&NiagaraDataInterfaceCurve, &NiagaraDataInterfaceArrayColor, &NiagaraDataInterfaceSkeletalMesh]
 usedDynamicInputs [&Multiply_LinearColorByFloat]
 usedModules [
  &self:SystemSpawnScript.SystemScriptSource.SystemScriptGraph.NiagaraNodeAssignment_0.SetVariables_#1
  &SystemState
 ]
}
emitters {
 Core_Ribbons {
  emitterUpdateScript [
   &EmitterState(Life Cycle Mode=::NewEnumerator1, Loop Duration=1, Recalculate Duration Each Loop=true, UseLoopDelay=false)
   &VC_FilteredSocketCount()
   &VC_RateQualityHandler(High Spawn Rate=60, Low Spawn Rate=30, Use Custom Rate?=true)
   &VC_SpawnRibbon_v1(Number of Ribbons=$Emitter.PortCount, SpawnRate=$Emitter.Spawn Rate)
  ]
  particleSpawnScript [
   &VC_PaintedVariantsOverride(
    Black=<rgba 0.01 0.01 0.01 1.25>
    BurntSienna=<rgba 0.5 0.2917 0 1>
    Cobalt=<rgba 0 1 3 2>
    Crimson=<rgba 1.5 0 0 1.25>
    DefaultColor=<rgba 2 0.2583 0.1 4>
    ForestGreen=<rgba 0.3 1.5 0 1.75>
    Grey=<rgba 0.35 0.35 0.35 1.25>
    Lime=<rgba 0.6667 1 0 1.5>
    Orange=<rgba 2 1 0 1.5>
    Pink=<rgba 3 0.9 2.475 1.5>
    Purple=<rgba 0.8719 0.2025 1.35 1.5>
    R1 Black:LinearColor=<rgba 0.01 0.01 0.01 1.25>
    R1 BurntSienna:LinearColor=<rgba 0.5 0.2917 0 10>
    R1 Cobalt:LinearColor=<rgba 0 1 3 2>
    R1 Crimson:LinearColor=<rgba 1.5 0 0 1.25>
    R1 DefaultColor=<rgba 0 3.3517 5 1>
    R1 ForestGreen:LinearColor=<rgba 0.3 1.5 0 1.75>
    R1 Grey:LinearColor=<rgba 0.35 0.35 0.35 1.25>
    R1 Lime:LinearColor=<rgba 0.6667 1 0 1.5>
    R1 Orange:LinearColor=<rgba 1.5 0.75 0 1.5>
    R1 Pink:LinearColor=<rgba 4 1.4 3.35 1.5>
    R1 Purple:LinearColor=<rgba 0.8719 0.2025 1.35 1.5>
    R1 Saffron:LinearColor=<rgba 2 1.6033 0.3 1.15>
    R1 SkyBlue:LinearColor=<rgba 0.2025 0.8336 1.35 1.5>
    R1 TitaniumWhite:LinearColor=<rgba 0.5 0.5 0.5 1>
    R2 Black:LinearColor=<rgba 0.005 0.005 0.005 1.5>
    R2 BurntSienna:LinearColor=<rgba 0.35 0.1458 0 1.35>
    R2 Cobalt:LinearColor=<rgba 0 0.7167 2.15 1.15>
    R2 Crimson:LinearColor=<rgba 1.35 0.0022 0 1.35>
    R2 DefaultColor:LinearColor=<rgba 0.0254 1.048 3 1>
    R2 ForestGreen:LinearColor=<rgba 0.3 1.5 0 1.75>
    R2 Grey:LinearColor=<rgba 0.35 0.35 0.35 1.25>
    R2 Lime:LinearColor=<rgba 1.1667 1.75 0 1.25>
    R2 Orange:LinearColor=<rgba 2.5 1.25 0 1.35>
    R2 Pink:LinearColor=<rgba 1 0.5 0.875 1.35>
    R2 Purple:LinearColor=<rgba 1.75 0 3 1.25>
    R2 Saffron:LinearColor=<rgba 2.25 1.8038 0.3375 1.15>
    R2 SkyBlue:LinearColor=<rgba 0 2.1667 3.25 1.25>
    R2 TitaniumWhite:LinearColor=<rgba 0 0 0 0>
    R3 Black:LinearColor=<rgba 0 0 0 0>
    R3 BurntSienna:LinearColor=<rgba 0 0 0 0>
    R3 Cobalt:LinearColor=<rgba 0 0 0 0>
    R3 Crimson:LinearColor=<rgba 0 0 0 0>
    R3 DefaultAlpha=0
    R3 DefaultColor:LinearColor=<rgba 0 0 0 0>
    R3 ForestGreen:LinearColor=<rgba 0 0 0 0>
    R3 Gold:LinearColor=<rgba 0 0 0 0>
    R3 Grey:LinearColor=<rgba 0 0 0 0>
    R3 Lime:LinearColor=<rgba 0 0 0 0>
    R3 Orange:LinearColor=<rgba 0 0 0 0>
    R3 Pink:LinearColor=<rgba 0 0 0 0>
    R3 Purple:LinearColor=<rgba 0 0 0 0>
    R3 Saffron:LinearColor=<rgba 0 0 0 0>
    R3 SkyBlue:LinearColor=<rgba 0 0 0 0>
    R3 TitaniumWhite:LinearColor=<rgba 1.25 1.25 1.25 1.25>
    Saffron=<rgba 2 1.6033 0.3 1.15>
    SkyBlue=<rgba 0.2025 0.8336 1.35 1.5>
    TitaniumWhite=<rgba 0.6375 0.6375 0.75 1>
    Use R1 Painted Overrides=true
    Use R2 Painted Overrides=true
    Use R3 Painted Overrides=true
   )
   &VC_InitializeRibbon_v1(
    Color=$Particles.Color
    Custom Offset=true
    Enable Skeletal Position=true
    Facing Axis:ENiagaraOrientationAxis=::YAxis
    Lifetime=0.825
    Local Vector Offset=&AddVectorToPosition(
     Position=/Niagara/DynamicInputs/TypeConversions/MakeCustomPositionFromBool(Bool:NiagaraBool=$User.bIsFrontendPreview, True Position:NiagaraPosition=<-10 0 0>, False Position:NiagaraPosition=<0 0 0>)
     Vector=$User.DTCarOffset
    )
    Offset Amount=1
    Ribbon Width=25
   )
   &AddVelocity(Velocity=<-2500 0 0>) {module=AddVelocity001}
   /SetVariables(Particles.R1_RibbonWidth=60, Particles.R2_RibbonWidth=95) {isSetParameters=true}
  ]
  particleUpdateScript [
   &ParticleState()
   &InheritSourceMovement(
    Applied Position Scale=&Multiply_VectorByFloat(
     Vector=<1 1 1>
     Float=&FloatFromCurve(FloatCurve=DataInterface() {propertyValues {Curve {keys [
          {interpMode=RCIM_Cubic, leaveTangent=-0.0051, tangentMode=RCTM_Break, time=0, value=1}
          {arriveTangent=-0.247, interpMode=RCIM_Cubic, tangentMode=RCTM_Break, time=1, value=0.9}
         ]}}})
    )
    ApplyForce=false
    ApplyPosition=true
    ApplyPositionFalloff=false
   )
   &SolveForcesAndVelocity(Clamp Velocity=false, Limit Acceleration=false)
   &VC_UpdateRibbon_v1(Enable Skeletal Position=true)
   &ScaleColor(
    Scale Alpha=&FloatFromCurve(FloatCurve=DataInterface() {propertyValues {Curve {keys [<at 0 1>,<at 1 0.1>]}}})
    ScaleA=true
    ScaleRGB=true
   )
   &DynamicMaterialParameters(
    Index 0 Param 1=&MakeCustomFloatFromBool(
     Bool=$User.bIsPainted
     True Float=&FloatFromCurve(FloatCurve=DataInterface() {propertyValues {Curve {keys [<at 0 6 RCIM_Cubic>,<at 1 0.75 RCIM_Cubic>]}}}, Scale Curve=$User.GlowOverride)
     False Float=&FloatFromCurve(FloatCurve=DataInterface() {propertyValues {Curve {keys [<at 0 0.5 RCIM_Cubic>,<at 0.15 0.8>,<at 1 1>]}}}, Scale Curve=10)
    )
    Module.Index0Param0Write=true
    Module.Index0Param1Write=true
    Module.Index0Param2Write=true
    Module.Index0Param3Write=true
   )
   &ScaleRibbonWidth(Ribbon Width Scale=&FloatFromCurve(FloatCurve=DataInterface() {propertyValues {Curve {keys [<at 0 1>,<at 1 0.25>]}}}))
  ]
  rendererClasses [&NiagaraRibbonRendererProperties]
  renderers [
   {rendererClass=&NiagaraRibbonRendererProperties}
   {rendererClass=&NiagaraRibbonRendererProperties, rendererIndex=1}
   {rendererClass=&NiagaraRibbonRendererProperties, rendererIndex=2}
   {rendererClass=&NiagaraRibbonRendererProperties, rendererIndex=3}
  ]
 }
 Dot_Sprites {
  emitterUpdateScript [
   &EmitterState()
   &VC_FilteredSocketCount()
   &SpawnRate(SpawnRate=100, Use Spawn Probability=false)
  ]
  particleSpawnScript [
   &VC_PaintedVariantsOverride(
    Black=<rgba 0.05 0.05 0.05 6>
    BurntSienna=<rgba 0.175 0.0613 0 4>
    Cobalt=<rgba 0 1.5 4.5 2.5>
    Crimson=<rgba 2.1 0 0.0262 2.85>
    DefaultColor=&SelectLinearColorFromArray(
     Array Sampling Mode=::NewEnumerator0
     Direct Array Index=?
     Color Selection Array=DataInterface() {propertyValues {ColorData [<rgba 0 0.5045 1 1>, <rgba 1 0.145 0.05 1>]}}
     Random Seed=0
    )
    ForestGreen=<rgba 0.45 2.25 0 2.25>
    Grey=<rgba 0.15 0.15 0.15 20>
    Lime=<rgba 2.3333 3.5 0 2.5>
    Orange=<rgba 3.5 1.75 0 2.5>
    Pink=<rgba 5 1 4 2>
    Purple=<rgba 0.8798 0.1 2 2.75>
    Saffron=<rgba 3 2.3175 0.075 3>
    SkyBlue=<rgba 0.1 1.3667 2 3>
    TitaniumWhite=<rgba 1.5 1.5 1.5 5>
   )
   &InitializeParticle(
    Color=$Particles.Color
    Lifetime Max=0.35
    Lifetime Min=0.25
    Lifetime Mode=::NewEnumerator1
    Sprite Rotation Mode=::NewEnumerator2
    Sprite Size Max=<8 60>
    Sprite Size Min=<8 50>
    Sprite Size Mode=::NewEnumerator2
    UsePositionOffset=false
    Write Material Random=true
   )
   &PartitionParticles(Number Of Partitions=$Emitter.PortCount)
   &SkeletalMeshLocation(
    Bone / Socket Index=$Particles.PartitionParticles.Partition
    Bone / Socket Sampling Mode=::NewEnumerator5
    MaskSpawnGroup=false
    Mesh Sampling Type=::NewEnumerator6
    Sampled Position Offset=&AddVectorToPosition(Position=<-70 0 0>, Vector=$User.DTCarOffset)
    Sampled Position Offset Coordinate Space:ENiagaraSkelMeshTransforms=::NewEnumerator2
    Skeletal Mesh=$System.SkeletalMesh
   )
   &AddVelocity(Velocity=&RandomRangeVector(Minimum=<-2000 100 -50>, Maximum=<-3000 -100 50>))
   &ShapeLocation(
    MaskSpawnGroup=false
    Override Seed=false
    Rotation Angle=0.25
    Rotation Axis=<0 1 0>
    Rotation Mode=::NewEnumerator1
    Sphere Radius=60
    UseOwnerScale=true
   )
  ]
  particleUpdateScript [
   &ParticleState()
   &ScaleSpriteSize(Uniform Curve Sprite Scale=DataInterface() {propertyValues {Curve {keys [
        {interpMode=RCIM_Cubic, leaveTangent=-0.1977, tangentMode=RCTM_Break, time=0, value=1}
        {arriveTangent=-1.1493, interpMode=RCIM_Cubic, leaveTangent=-1.1493, tangentMode=RCTM_User, time=0.5, value=0.85}
        {arriveTangent=-4.2625, interpMode=RCIM_Cubic, tangentMode=RCTM_Break, time=1, value=0}
       ]}}})
   &VortexForce(
    Origin Pull Amount:NiagaraFloat=&MakeCustomFloatFromBool(Bool=$User.bIsFrontendPreview, True Float=10000, False Float=10000)
    UseOriginOffset=true
    Vortex Axis=&RandomRangeVector(Minimum=<-1 0 0>, Maximum=<1 0 0>, Evaluation Type=::NewEnumerator1)
    Vortex Axis Coordinate Space:ENiagaraCoordinateSpace=::Local
    Vortex Force Amount=&MakeCustomFloatFromBool(Bool=$User.bIsFrontendPreview, True Float=10000, False Float=7500)
    Vortex Origin:NiagaraPosition=$Particles.Initial.Position
    Vortex Origin Offset:Vector3f=&RandomRangeVector(Minimum=<-1000 0 0>, Maximum=<-400 0 0>)
    Vortex Origin Offset Coordinate Space:ENiagaraCoordinateSpace=::Local
   )
   &InheritSourceMovement(
    Applied Position Scale=&Multiply_VectorByFloat(
     Vector=<1 1 1>
     Float=&FloatFromCurve(FloatCurve=DataInterface() {propertyValues {Curve {keys [{interpMode=RCIM_Cubic, tangentMode=RCTM_User, time=0.3, value=1}, {interpMode=RCIM_Cubic, tangentMode=RCTM_User, time=1, value=0}]}}})
    )
    ApplyForce=false
    ApplyPosition=true
    ApplyPositionFalloff=false
   )
   &CurlNoiseForce(Bias Noise=false, Noise Frequency=1, Noise Strength=50, Pan Noise=false)
   &SolveForcesAndVelocity(Clamp Velocity=false, Limit Acceleration=false)
   &ScaleColor(
    Scale Alpha=&FloatFromCurve(FloatCurve=DataInterface() {propertyValues {Curve {keys [
         <at 0.02 0>
         <at 0.2 1>
         {interpMode=RCIM_Cubic, tangentMode=RCTM_User, time=0.8, value=1}
         {interpMode=RCIM_Cubic, tangentMode=RCTM_User, time=1, value=0}
        ]}}})
    Scale RGB=&Multiply_VectorByFloat(Vector=<1 1 1>, Float=&MakeCustomFloatFromBool(Bool=$User.bIsPainted, True Float=1, False Float=100))
    ScaleA=true
    ScaleRGB=true
   )
  ]
  rendererClasses [&NiagaraSpriteRendererProperties]
  renderers [{rendererClass=&NiagaraSpriteRendererProperties}]
 }
 Dot_Sprites001 {
  emitterUpdateScript [
   &EmitterState()
   &VC_FilteredSocketCount()
   &SpawnRate(SpawnRate=100, Use Spawn Probability=false)
  ]
  particleSpawnScript [
   &VC_PaintedVariantsOverride(
    Black=<rgba 0.05 0.05 0.05 6>
    BurntSienna=<rgba 0.175 0.0613 0 4>
    Cobalt=<rgba 0 1.5 4.5 2.5>
    Crimson=<rgba 2.1 0 0.0262 2.85>
    DefaultColor=&SelectLinearColorFromArray(
     Array Sampling Mode=::NewEnumerator0
     Direct Array Index=?
     Color Selection Array=DataInterface() {propertyValues {ColorData [<rgba 0 0.5045 1 1>, <rgba 1 0.145 0.05 1>]}}
     Random Seed=0
    )
    ForestGreen=<rgba 0.45 2.25 0 2.25>
    Grey=<rgba 0.15 0.15 0.15 20>
    Lime=<rgba 2.3333 3.5 0 2.5>
    Orange=<rgba 3.5 1.75 0 2.5>
    Pink=<rgba 5 1 4 2>
    Purple=<rgba 0.8798 0.1 2 2.75>
    Saffron=<rgba 3 2.3175 0.075 3>
    SkyBlue=<rgba 0.1 1.3667 2 3>
    TitaniumWhite=<rgba 1.5 1.5 1.5 5>
   )
   &InitializeParticle(
    Color=$Particles.Color
    Lifetime Max=0.35
    Lifetime Min=0.25
    Lifetime Mode=::NewEnumerator1
    Sprite Rotation Mode=::NewEnumerator2
    Sprite Size Max=<6 50>
    Sprite Size Min=<6 40>
    Sprite Size Mode=::NewEnumerator2
    UsePositionOffset=false
    Write Material Random=true
   )
   &PartitionParticles(Number Of Partitions=$Emitter.PortCount)
   &SkeletalMeshLocation(
    Bone / Socket Index=$Particles.PartitionParticles.Partition
    Bone / Socket Sampling Mode=::NewEnumerator5
    MaskSpawnGroup=false
    Mesh Sampling Type=::NewEnumerator6
    Sampled Position Offset=&AddVectorToPosition(Position=<-70 0 0>, Vector=$User.DTCarOffset)
    Sampled Position Offset Coordinate Space:ENiagaraSkelMeshTransforms=::NewEnumerator2
    Skeletal Mesh=$System.SkeletalMesh
   )
   &AddVelocity(Velocity=&RandomRangeVector(Minimum=<-2000 100 -50>, Maximum=<-3000 -100 50>))
   &ShapeLocation(
    MaskSpawnGroup=false
    Override Seed=false
    Rotation Angle=0.25
    Rotation Axis=<0 0 1>
    Rotation Mode=::NewEnumerator1
    Sphere Radius=60
    UseOwnerScale=true
   )
  ]
  particleUpdateScript [
   &ParticleState()
   &ScaleSpriteSize(Uniform Curve Sprite Scale=DataInterface() {propertyValues {Curve {keys [
        {interpMode=RCIM_Cubic, leaveTangent=-0.1977, tangentMode=RCTM_Break, time=0, value=1}
        {arriveTangent=-1.1493, interpMode=RCIM_Cubic, leaveTangent=-1.1493, tangentMode=RCTM_User, time=0.5, value=0.85}
        {arriveTangent=-4.2625, interpMode=RCIM_Cubic, tangentMode=RCTM_Break, time=1, value=0}
       ]}}})
   &VortexForce(
    Origin Pull Amount:NiagaraFloat=&MakeCustomFloatFromBool(Bool=$User.bIsFrontendPreview, True Float=10000, False Float=10000)
    UseOriginOffset=true
    Vortex Axis=&RandomRangeVector(Minimum=<-1 0 0>, Maximum=<1 0 0>, Evaluation Type=::NewEnumerator1)
    Vortex Axis Coordinate Space:ENiagaraCoordinateSpace=::Local
    Vortex Force Amount=&MakeCustomFloatFromBool(Bool=$User.bIsFrontendPreview, True Float=10000, False Float=7500)
    Vortex Origin:NiagaraPosition=$Particles.Initial.Position
    Vortex Origin Offset:Vector3f=&RandomRangeVector(Minimum=<-1000 0 0>, Maximum=<-400 0 0>)
    Vortex Origin Offset Coordinate Space:ENiagaraCoordinateSpace=::Local
   )
   &InheritSourceMovement(
    Applied Position Scale=&Multiply_VectorByFloat(
     Vector=<1 1 1>
     Float=&FloatFromCurve(FloatCurve=DataInterface() {propertyValues {Curve {keys [{interpMode=RCIM_Cubic, tangentMode=RCTM_User, time=0.3, value=1}, {interpMode=RCIM_Cubic, tangentMode=RCTM_User, time=1, value=0}]}}})
    )
    ApplyForce=false
    ApplyPosition=true
    ApplyPositionFalloff=false
   )
   &CurlNoiseForce(Bias Noise=false, Noise Frequency=1, Noise Strength=50, Pan Noise=false)
   &SolveForcesAndVelocity(Clamp Velocity=false, Limit Acceleration=false)
   &ScaleColor(
    Scale Alpha=&FloatFromCurve(FloatCurve=DataInterface() {propertyValues {Curve {keys [
         <at 0.02 0>
         <at 0.2 1>
         {interpMode=RCIM_Cubic, tangentMode=RCTM_User, time=0.8, value=1}
         {interpMode=RCIM_Cubic, tangentMode=RCTM_User, time=1, value=0}
        ]}}})
    Scale RGB=&Multiply_VectorByFloat(Vector=<1 1 1>, Float=&MakeCustomFloatFromBool(Bool=$User.bIsPainted, True Float=1, False Float=100))
    ScaleA=true
    ScaleRGB=true
   )
  ]
  rendererClasses [&NiagaraSpriteRendererProperties]
  renderers [{rendererClass=&NiagaraSpriteRendererProperties}]
 }
 Spiral_Ribbons {
  emitterUpdateScript [
   &EmitterState(Life Cycle Mode=::NewEnumerator1, Loop Duration=1, Recalculate Duration Each Loop=true, UseLoopDelay=false)
   &VC_FilteredSocketCount()
   &VC_RateQualityHandler(High Spawn Rate=60, Low Spawn Rate=30, Use Custom Rate?=true)
   &VC_SpawnRibbon_v1(Number of Ribbons=$Emitter.PortCount, SpawnRate=$Emitter.Spawn Rate)
  ]
  particleSpawnScript [
   &VC_PaintedVariantsOverride(
    Black=<rgba 0.15 0.15 0.15 5>
    BurntSienna=<rgba 0.425 0.2125 0 2.15>
    Cobalt=<rgba 0 0.9167 2.75 1.15>
    Crimson=<rgba 1.15 0.0077 0 1.85>
    DefaultColor=<rgba 8 1.304 0.56 2>
    ForestGreen=<rgba 0.25 1.25 0 2.75>
    Grey=<rgba 0.375 0.375 0.375 3.5>
    Lime=<rgba 1.3333 2 0 1.5>
    Orange=<rgba 3 1.25 0 1.75>
    Pink=<rgba 3 0.9 2.475 1.75>
    Purple=<rgba 0.86 0.1 2 2.25>
    R1 Black:LinearColor=<rgba 0.15 0.15 0.15 5>
    R1 BurntSienna:LinearColor=<rgba 0.425 0.2125 0 2.15>
    R1 Cobalt:LinearColor=<rgba 0 0.9167 2.75 1.15>
    R1 Crimson:LinearColor=<rgba 1.15 0.0077 0 1.85>
    R1 DefaultColor=<rgba 0 1.6667 10 1>
    R1 ForestGreen:LinearColor=<rgba 0.25 1.25 0 2.75>
    R1 Grey:LinearColor=<rgba 0.375 0.375 0.375 3.5>
    R1 Lime:LinearColor=<rgba 1.3333 2 0 1.5>
    R1 Orange:LinearColor=<rgba 3 1.25 0 1.75>
    R1 Pink:LinearColor=<rgba 3 0.9 2.475 1.75>
    R1 Purple:LinearColor=<rgba 0.86 0.1 2 2.25>
    R1 Saffron:LinearColor=<rgba 2.25 1.7644 0.1687 1.15>
    R1 SkyBlue:LinearColor=<rgba 0.1 1.3667 2 2.25>
    R1 TitaniumWhite:LinearColor=<rgba 2 2 2 1.75>
    Saffron=<rgba 2.25 1.7644 0.1687 1.15>
    SkyBlue=<rgba 0.1 1.3667 2 2.25>
    TitaniumWhite=<rgba 2 2 2 1.75>
    Use R1 Painted Overrides=true
   )
   &VC_InitializeRibbon_v1(
    Color=$Particles.Color
    Custom Offset=true
    Enable Skeletal Position=true
    Facing Axis:ENiagaraOrientationAxis=::YAxis
    Lifetime=0.875
    Local Vector Offset=&AddVectorToPosition(
     Position=&MakeCustomPositionFromBool(Bool:NiagaraBool=$User.bIsFrontendPreview, True Position:NiagaraPosition=<-10 0 0>, False Position:NiagaraPosition=<0 0 0>)
     Vector=$User.DTCarOffset
    )
    Offset Amount=1
    Ribbon Width=155
   )
   &AddVelocity(Velocity=<-2500 0 0>) {module=AddVelocity001}
  ]
  particleUpdateScript [
   &ParticleState()
   &InheritSourceMovement(
    Applied Position Scale=&Multiply_VectorByFloat(
     Vector=<1 1 1>
     Float=&FloatFromCurve(FloatCurve=DataInterface() {propertyValues {Curve {keys [
          {interpMode=RCIM_Cubic, leaveTangent=-0.0051, tangentMode=RCTM_Break, time=0, value=1}
          {arriveTangent=-0.247, interpMode=RCIM_Cubic, tangentMode=RCTM_Break, time=1, value=0.9}
         ]}}})
    )
    ApplyForce=false
    ApplyPosition=true
    ApplyPositionFalloff=false
   )
   &SolveForcesAndVelocity(Clamp Velocity=false, Limit Acceleration=false)
   &VC_UpdateRibbon_v1(Enable Skeletal Position=true)
   &ScaleColor(ScaleA=true, ScaleRGB=true)
   &DynamicMaterialParameters(
    Index 0 Param 1=&MakeCustomFloatFromBool(
     Bool=$User.bIsPainted
     True Float=&FloatFromCurve(FloatCurve=DataInterface() {propertyValues {Curve {keys [<at 0 6 RCIM_Cubic>,<at 1 0.75 RCIM_Cubic>]}}}, Scale Curve=$User.GlowOverride)
     False Float=&FloatFromCurve(FloatCurve=DataInterface() {propertyValues {Curve {keys [<at 0 0.5 RCIM_Cubic>,<at 0.15 0.8>,<at 1 1>]}}}, Scale Curve=10)
    )
    Module.Index0Param0Write=true
    Module.Index0Param1Write=true
    Module.Index0Param2Write=true
    Module.Index0Param3Write=true
   )
   &ScaleRibbonWidth(Ribbon Width Scale=&FloatFromCurve(FloatCurve=DataInterface() {propertyValues {Curve {keys [<at 0 1>,<at 1 0.75>]}}}))
  ]
  rendererClasses [&NiagaraRibbonRendererProperties]
  renderers [
   {rendererClass=&NiagaraRibbonRendererProperties}
   {rendererClass=&NiagaraRibbonRendererProperties, rendererIndex=1}
  ]
 }
}
summary {
 systemName=NS_Boost_Energywave
 emitters [
  {emitterName=Core_Ribbons, rendererClasses [&NiagaraRibbonRendererProperties]}
  {emitterName=Dot_Sprites, rendererClasses [&NiagaraSpriteRendererProperties]}
  {emitterName=Spiral_Ribbons, rendererClasses [&NiagaraRibbonRendererProperties]}
  {emitterName=Dot_Sprites001, rendererClasses [&NiagaraSpriteRendererProperties]}
 ]
 userVariables [
  User.bIsFrontendPreview {type=NiagaraBool}
  User.bIsPainted {type=NiagaraBool}
  User.ColorVariantID {type=NiagaraInt32}
  User.DTCarOffset {type=Vector3f, description="", defaultValue {struct=Vector3f, value=<0 0 0>}}
  User.GlowOverride {type=NiagaraFloat, defaultValue=4, description=""}
  User.PaintedColor {type=LinearColor, description="", defaultValue {struct=LinearColor, value=<rgba 1 1 1 1>}}
  User.PaintedColor_2 {type=LinearColor, description="", defaultValue {struct=LinearColor, value=<rgba 1 1 1 1>}}
  User.PaintedColor_3 {type=LinearColor, description="", defaultValue {struct=LinearColor, value=<rgba 1 1 1 1>}}
  User.PaintedColor_4 {type=LinearColor, description="", defaultValue {struct=LinearColor, value=<rgba 1 1 1 1>}}
  User.PaintedColor_5 {type=LinearColor, description="", defaultValue {struct=LinearColor, value=<rgba 1 1 1 1>}}
  User.PaintedScalar {type=NiagaraFloat, defaultValue=1, description=""}
  User.PortCount {type=NiagaraInt32, defaultValue=1}
 ]
}
